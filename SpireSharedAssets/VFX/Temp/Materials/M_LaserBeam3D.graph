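# Unlit masked VFX material for a 3D laser beam, usable on particle meshes, ribbons and sprites.
# A tiled vertical gradient and contrast masks blend layered colors, tinted by particle color.
# A lattice mask pans rotated noise, distorted by adding Voronoi foam to the panner's time.
# Panned noise builds normal-direction vertex offset, scaled by dynamic parameter, alongside erosion, depth fade and near-camera fade.
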Material /SpireSharedAssets/VFX/Temp/Materials/M_LaserBeam3D
graph {
 properties {
  bUsedWithMeshDeformer=true
  bUsedWithMeshParticles=true
  bUsedWithNiagaraMeshParticles=true
  bUsedWithNiagaraRibbons=true
  bUsedWithParticleSprites=true
  bUsedWithSkeletalMesh=true
  blendMode=BLEND_Masked
  pixelDepthOffsetMode=PDOM_Legacy
  shadingModel=MSM_Unlit
  usedShadingModels=Unlit
 }
 nodes [
  n1 = NamedRerouteUsage()
  MaterialFunctionCall(
   In=Add(A=OneMinus(None=@n1), B=DynamicParameter() {parameterIndex=1, paramNames [Erosion,NA,NA,NA]}.Erosion)
   Contrast=ScalarParameter() {defaultValue=1, parameterName=Erosion Contrast}
  ) {materialFunction=&CheapContrast}
  n2 = VectorParameter() {parameterName=Color Outside, defaultValue=<rgba 0 0.2251 0.6042 1>}
  n3 = ScalarParameter() {defaultValue=2, group=Main Gradient, parameterName=GradientTiling}
  n4 = TextureSample(UVs=Multiply(A=TextureCoordinate(), B=AppendVector(A=@n3, B=@n3))) {samplerType=SAMPLERTYPE_Masks, texture=/Game/Global/Textures/Default/Masks_and_Gradients/T_VertGradient_MASK}
  n5 = MaterialFunctionCall(
   In=Add(
    A=NamedRerouteDeclaration(None=@n4.R) {name=Tiling, nodeColor=<rgba 1 0.7266 0 1>}
    B=ScalarParameter() {defaultValue=-0.7, group=Main Gradient, parameterName=Color Offset 1}
   ) {constB=0.2}
   Contrast=ScalarParameter() {group=Main Gradient, parameterName=Color Contrast 1}
  ) {materialFunction=&CheapContrast}
  n6 = LinearInterpolate(
   A=LinearInterpolate(
    A=VectorParameter() {parameterName=Color Wave Dark, defaultValue=<rgba 0 0.4181 1.2 1>}.RGB
    B=VectorParameter() {parameterName=Color Mid, defaultValue=<rgba 0 0.7453 2 1>}.RGB
    Alpha=NamedRerouteUsage()
   )
   B=LinearInterpolate(
    A=VectorParameter() {parameterName=Color Wave Bright, defaultValue=<rgba 0.408 3.9258 5 1>}.RGB
    B=VectorParameter() {parameterName=Color Center, defaultValue=<rgba 3 3 3 0>}.RGB
    Alpha=NamedRerouteUsage()
   )
   Alpha=@n5.Result
  )
  n7 = MaterialFunctionCall(
   In=Add(A=NamedRerouteUsage(), B=ScalarParameter() {group=Main Gradient, parameterName=Color Offset 2}) {constB=0.2}
   Contrast=ScalarParameter() {defaultValue=10, group=Main Gradient, parameterName=Color Contrast 2}
  ) {materialFunction=&CheapContrast}
  MaterialFunctionCall(Emissive=Multiply(A=ParticleColor().RGB, B=LinearInterpolate(A=@n2.RGB, B=@n6, Alpha=@n7.Result))) {materialFunction=&MF_FX_EmissiveAdjustments}
  Multiply(A=@n5.Result, B=Add(A=NamedRerouteUsage(), B=Constant() {r=0.25}))
  n8 = MaterialFunctionCall(Distance=ScalarParameter() {defaultValue=100, group=Alpha, parameterName=NearCameraFade}) {materialFunction=&Near_Camera_Fade}
  n9 = ScalarParameter() {defaultValue=50, group=Alpha, parameterName=FadeDistance}
  n10 = ScalarParameter() {parameterName=Alpha Outside}
  n11 = NamedRerouteUsage()
  n12 = ScalarParameter() {defaultValue=0.2, group=Alpha, parameterName=Alpha Offset 1}
  n13 = LinearInterpolate(
   A=ScalarParameter() {defaultValue=1, group=Alpha, parameterName=Alpha Mid}
   B=ScalarParameter() {defaultValue=1, group=Alpha, parameterName=Alpha Center}
   Alpha=MaterialFunctionCall(In=Add(A=@n11, B=@n12) {constB=0.2}, Contrast=ScalarParameter() {group=Alpha, parameterName=Alpha Contrast 1}) {materialFunction=&CheapContrast}.Result
  )
  n14 = MaterialFunctionCall(
   In=Add(A=NamedRerouteUsage(), B=ScalarParameter() {group=Alpha, parameterName=Alpha Offset 2}) {constB=0.2}
   Contrast=ScalarParameter() {group=Alpha, parameterName=Alpha Contrast 2}
  ) {materialFunction=&CheapContrast}
  Multiply(B=Multiply(A=DepthFade(Opacity=@n8.Result, FadeDistance=@n9) {fadeDistanceDefault=50}, B=LinearInterpolate(A=@n10, B=@n13, Alpha=@n14.Result)))
  n15 = TextureCoordinate()
  n16 = ComponentMask(None=@n15) {r=true}
  n17 = ScalarParameter() {defaultValue=1, group=WPO, parameterName=WPO Tile U}
  n18 = ComponentMask(None=@n15) {g=true}
  n19 = ScalarParameter() {defaultValue=1, group=WPO, parameterName=WPO Tile V}
  n20 = Panner(
   Coordinate=AppendVector(A=Multiply(A=@n16, B=@n17), B=Multiply(A=@n18, B=@n19))
   Time=Time() {bOverride_Period=true, period=64}
   Speed=AppendVector(
    A=ScalarParameter() {defaultValue=-1.8, group=WPO, parameterName=WPO Pan Speed U}
    B=ScalarParameter() {defaultValue=0.7, group=WPO, parameterName=WPO Pan Speed V}
   )
  ) {speedX=-1, speedY=0.1}
  n21 = Multiply(
   A=VertexNormalWS()
   B=MaterialFunctionCall(
    In=TextureSample(UVs=@n20) {samplerType=SAMPLERTYPE_Grayscale, texture=/SpireSharedAssets/VFX/Temp/Materials/FortWater/Textures/LowResBlurredNoise}.R
    Contrast=ScalarParameter() {defaultValue=0.3, parameterName=NoiseContrast}
   ) {materialFunction=&CheapContrast}.Result
  )
  n22 = ScalarParameter() {defaultValue=10, group=WPO, parameterName=WPO Amnt}
  NamedRerouteDeclaration(None=Multiply(A=Multiply(A=@n21, B=@n22), B=DynamicParameter() {paramNames [WPO Offset, WPO Amount, NA, NA]}.WPO Amount)) {name=WPO, nodeColor=<rgba 0 1 0.6484 1>}
  n23 = NamedRerouteDeclaration(None=MaterialFunctionCall(
    UVs=TextureCoordinate()
    Rotation Center=Constant2Vector() {g=1, r=0.5}
    "Rotation Angle (0-1)"=ScalarParameter() {defaultValue=0.1, parameterName=UV Rotation Angle}
   ) {materialFunction=&CustomRotator}.Rotated Values) {name=Lattice UV Rotation, nodeColor=<rgba 1 0.0938 0 1>}
  n24 = Multiply(A=ComponentMask(None=@n23) {r=true}, B=ScalarParameter() {defaultValue=0.7, group=Distortion, parameterName=Noise Tile U})
  n25 = Multiply(A=ComponentMask(None=@n23) {g=true}, B=ScalarParameter() {defaultValue=1.2, group=Distortion, parameterName=Noise Tile V})
  n26 = NamedRerouteUsage()
  n27 = AppendVector(
   A=Multiply(A=ComponentMask(None=@n26) {r=true}, B=ScalarParameter() {defaultValue=0.5, group=Distortion, parameterName=Distort Tile U})
   B=Multiply(A=ComponentMask(None=@n26) {g=true}, B=ScalarParameter() {defaultValue=0.5, group=Distortion, parameterName=Distort Tile V})
  )
  n28 = NamedRerouteUsage()
  n29 = AppendVector(
   A=ScalarParameter() {defaultValue=-0.45, group=Mask, parameterName=Distort Pan Speed U}
   B=ScalarParameter() {group=Mask, parameterName=Distort Pan Speed V}
  )
  n30 = OneMinus(None=TextureSample(UVs=Panner(Coordinate=@n27, Time=@n28, Speed=@n29) {speedY=0.1}) {samplerType=SAMPLERTYPE_LinearColor, texture=/Game/Athena/Apollo/Water/Textures/02_Rounded/02_WaterFoam_Veroni_base_H}.RGB)
  n31 = ScalarParameter() {parameterName=Distort Bias}
  n32 = Multiply(
   A=MaterialFunctionCall(In=Add(A=@n30, B=@n31), Contrast=ScalarParameter() {parameterName=Distort Contrast}) {materialFunction=&CheapContrast}.Result
   B=ScalarParameter() {defaultValue=0.15, group=Distortion, parameterName=Distort Strenght}
  ) {constB=2}
  n33 = NamedRerouteDeclaration(None=Add(A=Time(), B=ScalarParameter() {group=Mask, parameterName=UV Offset})) {name=Time, nodeColor=<rgba 0 0.1328 1 1>}
  n34 = ScalarParameter() {defaultValue=-0.7, group=Mask, parameterName=Pan Speed U}
  n35 = ScalarParameter() {group=Mask, parameterName=Pan Speed V}
  NamedRerouteDeclaration(None=Panner(Coordinate=AppendVector(A=@n24, B=@n25), Time=Add(A=@n32, B=@n33), Speed=AppendVector(A=@n34, B=@n35)) {speedY=0.1}) {name=Distorted UVs, nodeColor=<rgba 1 0 0.7031 1>}
  n36 = OneMinus(None=TextureSample(UVs=NamedRerouteUsage()) {samplerType=SAMPLERTYPE_Grayscale, texture=/Game/Global/Textures/Default/Noise/TilingNoise06}.RGB)
  n37 = ScalarParameter() {parameterName=Lattice Bias}
  NamedRerouteDeclaration(None=MaterialFunctionCall(In=Add(A=@n36, B=@n37), Contrast=ScalarParameter() {defaultValue=0.1, parameterName=Lattice Contrast}) {materialFunction=&CheapContrast}.Result) {name=Lattice, nodeColor=<rgba 1 0 0.6328 1>}
  NamedRerouteUsage()
  TextureCoordinate() {uTiling=0.9, vTiling=0.9}
  VectorParameter() {parameterName=Color Mid, defaultValue=<rgba 0 0.7453 2 1>}
  VectorParameter() {parameterName=Color Center, defaultValue=<rgba 3 3 3 0>}
 ]
}
