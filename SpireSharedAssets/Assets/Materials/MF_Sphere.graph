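# Material function that projects a procedural sphere onto 2D UVs, outputting normals, a clamped normal, a mask, height and spherified UVs.
# It centers the UV, divides by Size and scales by Curvature, then derives height as the square root of one minus squared radius.
# The clamped normal falls back to straight up outside the Ceil-based mask.
# Spherified UVs divide centered coordinates by offset height and remap them.
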MaterialFunction /SpireSharedAssets/Assets/Materials/MF_Sphere
graph {nodes [
  n1 = FunctionInput(Preview=TextureCoordinate()) {bUsePreviewValueAsDefault=true, inputName=UV, inputType=FunctionInput_Vector2, previewValue=<1 1 0 1>}
  n2 = FunctionInput() {bUsePreviewValueAsDefault=true, inputName=Size, inputType=FunctionInput_Vector2, sortPriority=1, previewValue=<1 1 0 1>}
  n3 = Divide(A=ConstantBiasScale(None=@n1) {bias=-0.5, scale=2}, B=@n2)
  n4 = FunctionInput() {bUsePreviewValueAsDefault=true, inputName=Curvature, inputType=FunctionInput_Vector2, sortPriority=1, previewValue=<1 1 0 1>}
  n5 = Multiply(A=@n3, B=@n4)
  n6 = DotProduct(A=@n5, B=@n5)
  n7 = SquareRoot(None=OneMinus(None=Saturate(None=@n6)))
  FunctionOutput(None=AppendVector(A=@n5, B=@n7)) {description=Spherical normals, outputName=Normal}
  n8 = Ceil(None=@n7)
  n9 = OneMinus(None=@n8)
  FunctionOutput(None=AppendVector(A=Multiply(A=@n5, B=@n8), B=Add(A=@n9, B=@n7))) {description="Cost a few instructions more, gives a 'safe' result", outputName=Clamped Normal, sortPriority=1}
  FunctionOutput(None=@n8) {description=Mask of the sphere, outputName=Mask, sortPriority=3}
  FunctionOutput(None=@n7) {description=Normalized height of the sphere, outputName=Height, sortPriority=4}
  n10 = Add(A=@n7)
  FunctionOutput(None=ConstantBiasScale(None=Divide(A=@n3, B=@n10))) {description=Spherified UVs, outputName=UV, sortPriority=5}
 ]}
