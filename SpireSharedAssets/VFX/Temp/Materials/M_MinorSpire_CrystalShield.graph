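# Translucent unlit crystal shield material for the Minor Spire VFX, usable on Niagara mesh particles.
# Emissive blends three panning chromatic caustics, warped by a turbulence flowmap, with a hue-shifted tint, dimmed at night.
# Opacity multiplies flowmapped Fresnel, UV band and rhombus masks, rotating-axis erosion, depth and view-angle fades.
# Particle dynamic parameters drive erosion and a particle colour/alpha override; Low/Medium quality drops flowmap, depth fade and extra caustic samples.
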Material /SpireSharedAssets/VFX/Temp/Materials/M_MinorSpire_CrystalShield
graph {
 properties {
  bUsedWithNanite=true
  bUsedWithNiagaraMeshParticles=true
  blendMode=BLEND_Translucent
  pixelDepthOffsetMode=PDOM_Legacy
  shadingModel=MSM_Unlit
  usedShadingModels=Unlit
 }
 nodes [
  Constant() {r=2}
  n1 = ScalarParameter() {defaultValue=0.111, parameterName=Intersection Opacity Softness}
  n2 = CameraVectorWS()
  n3 = CrossProduct(A=@n2, B=Transform(None=Constant3Vector() {constant=<rgba 0 1 0 0>}) {transformSourceType=TRANSFORMSOURCE_View})
  n4 = MaterialFunctionCall!(
   UV=MaterialFunctionCall(
    Plane Normal Axis=@n2
    Plane X Axis=Normalize(VectorInput=@n3)
    Plane Center=MaterialFunctionCall() {materialFunction=&ComponentPivotLocation}.Result
    UV Scale=ScalarParameter() {defaultValue=400, parameterName=UV Scale}
   ) {materialFunction=&VirtualPlaneCoordinates}.UVs
   Size
   Curvature
  ) {materialFunction=/SpireSharedAssets/Assets/Materials/MF_Sphere}
  n5 = ScalarParameter() {defaultValue=0.5, parameterName=Flowmap UV Scale, sliderMax=1}
  n6 = Constant()
  n7 = Multiply(A=ScalarParameter() {defaultValue=0.1, parameterName=Flowmap Panner Speed}, B=TwoSidedSign())
  n8 = TextureSample(UVs=Panner(Coordinate=Multiply(A=@n4.UV, B=@n5), Speed=AppendVector(A=@n6, B=@n7)) {bFractionalPart=true}) {samplerType=SAMPLERTYPE_LinearColor, texture=/Game/Effects/Fort_Effects/Textures/Plasma/T_Turbulence_AdditiveWarp_Linear}
  n9 = DotProduct(A=VertexNormalWS(), B=CameraVectorWS())
  n10 = Multiply(A=ConstantBiasScale(None=ComponentMask(None=@n8.RGB) {g=true, r=true}) {bias=-0.5, scale=2}, B=SquareRoot(None=Abs(None=@n9)))
  n11 = Multiply(A=@n10, B=ScalarParameter() {defaultValue=0.333, parameterName=Fresnel Flowmap Intensity})
  n12 = Constant()
  n13 = VertexNormalWS()
  n14 = Add(A=Transform(None=AppendVector(A=@n11, B=@n12)) {transformSourceType=TRANSFORMSOURCE_View}, B=@n13)
  n15 = DotProduct(A=QualitySwitch(Default=Normalize(VectorInput=@n14), Low=@n13, Medium=@n13) {desc=Skipping flowmap on med/low}, B=CameraVectorWS())
  n16 = Power(Base=OneMinus(None=Abs(None=@n15)), Exp=ScalarParameter() {defaultValue=3, parameterName=Center Opacity Exponent})
  n17 = OneMinus(None=@n16)
  n18 = OneMinus(None=Saturate(None=ScalarParameter() {defaultValue=0.1109, parameterName=Center Opacity}))
  n19 = Sine(None=ComponentMask(None=TextureCoordinate() {coordinateIndex=1, vTiling=0.5}) {g=true})
  n20 = ScalarParameter() {defaultValue=1, parameterName=Exp}
  n21 = ScalarParameter() {defaultValue=70, parameterName=B}
  n22 = Sine(None=ComponentMask(None=TextureCoordinate() {coordinateIndex=1, uTiling=0.5, vTiling=0.5}) {r=true})
  n23 = ConstantBiasScale(None=ComponentMask(None=TextureCoordinate() {coordinateIndex=1}) {r=true}) {bias=-0.5, scale=2}
  n24 = Power(Base=Min(A=Power(Base=@n22) {constExponent=4}, B=DotProduct(A=@n23, B=@n23)), Exp=ScalarParameter() {defaultValue=2})
  n25 = Multiply(
   A=OneMinus(None=Multiply(A=@n17, B=@n18))
   B=Multiply(A=Power(Base=@n19, Exp=@n20) {constExponent=8}, B=Multiply(A=@n21, B=@n24) {constB=10})
  )
  n26 = MaterialFunctionCall(Exponent=ScalarParameter() {defaultValue=1.1909, parameterName=Param_1}) {materialFunction=&MF_Fresnel_FX}
  n27 = MaterialFunctionCall!(
   RhombusSize=VectorParameter() {parameterName="RhombusSize (V2)", defaultValue=<rgba -0.1076 0.5 0 0>}.RGB
   UVs=TextureCoordinate() {coordinateIndex=1}
   DistortionIntensity
  ) {materialFunction=/SpireSharedAssets/Assets/Materials/MF_Rhombus}
  n28 = ScalarParameter() {defaultValue=0.5, parameterName=Erosion Axis Rotation}
  n29 = Multiply(A=Time(), B=ScalarParameter() {defaultValue=0.666, parameterName=Erosion Axis Rotation Speed})
  n30 = AppendVector(A=Sine(None=@n29), B=Cosine(None=@n29))
  n31 = Add(
   A=VectorParameter() {parameterName=Erosion Axis, defaultValue=<rgba 0 0.5 1 0>}.RGB
   B=AppendVector(A=Multiply(A=@n28, B=@n30), B=Constant())
  )
  n32 = WorldPosition()
  n33 = MaterialFunctionCall() {materialFunction=&ComponentPivotLocation}
  n34 = Divide(A=DotProduct(A=Normalize(VectorInput=@n31), B=Subtract(A=@n32.XYZ, B=@n33.Result))) {constB=100}
  n35 = OneMinus(None=Saturate(None=Add(A=@n34) {constB=0.5}))
  n36 = ComponentMask(None=@n10) {r=true}
  n37 = ScalarParameter() {defaultValue=0.333, parameterName=Flowmap Erosion Intensity}
  n38 = Add(A=@n35, B=Multiply(A=ConstantBiasScale(None=@n36), B=@n37))
  n39 = Add(A=@n37) {constB=0.85, desc=eyeballed range of the flowmap texture}
  n40 = MaterialFunctionCall!(
   Gradient=QualitySwitch(Default=Divide(A=@n38, B=@n39), Low=@n35, Medium=@n35) {desc=Skipping flowmap on med/low}
   Erosion Amount=DynamicParameter() {defaultValue=<rgba 0 0 0 0>, paramNames [Erosion, Use Particle Color, - unused -, - unused -]}.Erosion
   Softness=ScalarParameter() {defaultValue=0.05, parameterName=Erosion Softness, sliderMax=1}
  ) {materialFunction=/Game/Effects/Fort_Effects/MaterialFunctions/MF_Erosion}
  n41 = Multiply(A=Multiply(A=Multiply(A=@n25, B=@n26.Fresnel), B=Multiply(A=@n27.InvertedMask) {constB=4}), B=@n40.Increased Range)
  n42 = Multiply(A=Saturate(None=Divide(B=@n1)) {desc=Depth Fade}, B=@n41)
  n43 = MaterialFunctionCall("FadeSensitivity (optional)"=ScalarParameter() {defaultValue=0.4708, parameterName="FadeSensitivity (optional) (S)"}) {materialFunction=&FadeBasedOnViewAngle}
  n44 = ParticleColor()
  n45 = DynamicParameter() {defaultValue=<rgba 0 0 0 0>, paramNames [Erosion, Use Particle Color, - unused -, - unused -]}
  LinearInterpolate(
   A=Multiply(A=QualitySwitch(Default=@n42, Low=@n41, Medium=@n41) {desc=Skipping depth fade on med/low}, B=OneMinus(None=@n43.Result))
   B=@n44.A
   Alpha=@n45.Use Particle Color
  )
  n46 = ScalarParameter() {defaultValue=0.45, parameterName=Nighttime Brightness}
  n47 = CollectionParameter() {
   collection=&FortniteMaterialParameters
   parameterName=FortniteActiveTimeOfDay
  }
  n48 = OneMinus(None=ComponentMask(None=@n47) {a=true, desc="1 when night, 0 when anything else"})
  n49 = Multiply(A=@n48, B=@n48)
  n50 = Add(A=@n4.UV, B=Multiply(A=ScalarParameter() {defaultValue=0.1, parameterName=UV Flowmap Intensity}, B=@n10))
  n51 = Transform(None=VertexNormalWS()) {transformSourceType=TRANSFORMSOURCE_World, transformType=TRANSFORM_Camera}
  n52 = VertexInterpolator(VS=ComponentMask(None=ConstantBiasScale(None=@n51)) {g=true, r=true})
  n53 = TwoSidedSign()
  n54 = Saturate(None=@n53)
  n55 = Multiply(A=@n54) {constB=0.25}
  n56 = MaterialFunctionCall!(
   UV=Divide(A=QualitySwitch(Default=@n50, Low=@n52.PS, Medium=@n52.PS), B=OneMinus(None=@n55))
   Rotation=Multiply(A=Constant() {r=0.25}, B=@n54)
   Speed=Multiply(A=@n53, B=ScalarParameter() {defaultValue=0.1, parameterName=Caustics Speed})
   Time
   Rotate UVs
   Fractional Part
  ) {materialFunction=/SpireSharedAssets/Assets/Materials/MF_3UVPanner}
  n57 = TextureSample(UVs=@n56.UV1) {texture=/Game/Effects/Fort_Effects/Textures/Tile/T_ShieldCaustics_Tile_Chromatic}
  n58 = Add(
   A=Add(A=@n57.RGB, B=TextureSample(UVs=@n56.UV2) {texture=&T_ShieldCaustics_Tile_Chromatic}.RGB)
   B=TextureSample(UVs=@n56.UV3) {texture=&T_ShieldCaustics_Tile_Chromatic}.RGB
  )
  n59 = OneMinus(None=QualitySwitch(Default=Divide(A=@n58) {constB=3}, Low=@n57.RGB, Medium=@n57.RGB) {desc=1 texture for med/low})
  n60 = OneMinus(None=Saturate(None=ScalarParameter() {defaultValue=0.0333, parameterName=Caustics Minimum Brightness})) {desc=Minimum brightness}
  n61 = Multiply(A=@n60) {desc=Brightens the near depth}
  n62 = QualitySwitch(
   Default=Multiply(A=Multiply(A=@n61, B=@n40.Increased Range) {desc=Adds a bright edge along the erosion}) {desc=Brightens up pixels inside sphere}
   Low=@n60
   Medium=@n60
  ) {desc=Skipping samplers on med/low}
  n63 = OneMinus(None=@n16)
  n64 = VectorParameter() {parameterName="Color && Hue Shift", defaultValue=<rgba 4.0728 2.2535 40 0.07>}
  n65 = Multiply(
   A=OneMinus(None=Multiply(A=@n59, B=@n62) {desc="All these combined\r\nbrighten this texture a bit"})
   B=MaterialFunctionCall(Hue Shift Percentage=Multiply(A=@n63, B=@n64.A), Texture=@n64.RGB) {materialFunction=&HueShift}.Result
  )
  MaterialFunctionCall(Emissive=Multiply(A=LinearInterpolate(A=@n46, Alpha=@n49) {constA=1}, B=LinearInterpolate(A=@n65, B=@n44.RGB, Alpha=@n45.Use Particle Color))) {materialFunction=&MF_FX_EmissiveAdjustments}
  Multiply()
  Multiply()
  Multiply(A=VertexNormalWS(), B=Constant() {r=1})
 ]
}
